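# Material function for glTF exporter proxy materials that applies a texture transform and tiling mode to a mesh UV channel.
# A TexCoord switch picks UV0–3, which is scaled, rotated about the origin through MF_GLTF_RotateV2, then offset.
# Custom HLSL applies per-axis wrap or mirror, passing clamp through unchanged.
# It also outputs DDX and DDY taken before tiling, so derivatives stay free of wrap discontinuities.
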MaterialFunction /GLTFExporter/Functions/Proxy/MF_GLTF_UVTransform
graph {nodes [
  MaterialExpressionTextureCoordinate_1 = TextureCoordinate()
  MaterialExpressionTextureCoordinate_5 = TextureCoordinate() {coordinateIndex=1}
  MaterialExpressionTextureCoordinate_6 = TextureCoordinate() {coordinateIndex=2}
  MaterialExpressionTextureCoordinate_7 = TextureCoordinate() {coordinateIndex=3}
  n1 = Switch!(
   SwitchValue=FunctionInput() {bUsePreviewValueAsDefault=true, inputName=TexCoord, inputType=FunctionInput_Scalar}
   Default
   UV0=@MaterialExpressionTextureCoordinate_1
   UV1=@MaterialExpressionTextureCoordinate_5
   UV2=@MaterialExpressionTextureCoordinate_6
   UV3=@MaterialExpressionTextureCoordinate_7
  ) {inputs [
    {
     inputName=UV0
     input {
      expression=&self:MaterialExpressionTextureCoordinate_1
      inputName=None
      mask=0
      maskA=0
      maskB=0
      maskG=0
      maskR=0
      outputIndex=0
     }
    }
    {
     inputName=UV1
     input {
      expression=&self:MaterialExpressionTextureCoordinate_5
      inputName=None
      mask=0
      maskA=0
      maskB=0
      maskG=0
      maskR=0
      outputIndex=0
     }
    }
    {
     inputName=UV2
     input {
      expression=&self:MaterialExpressionTextureCoordinate_6
      inputName=None
      mask=0
      maskA=0
      maskB=0
      maskG=0
      maskR=0
      outputIndex=0
     }
    }
    {
     inputName=UV3
     input {
      expression=&self:MaterialExpressionTextureCoordinate_7
      inputName=None
      mask=0
      maskA=0
      maskB=0
      maskG=0
      maskR=0
      outputIndex=0
     }
    }
   ]}
  n2 = FunctionInput() {bUsePreviewValueAsDefault=true, inputName=Scale_X, inputType=FunctionInput_Scalar, sortPriority=3, previewValue=<1 0 0 0>}
  n3 = FunctionInput() {bUsePreviewValueAsDefault=true, inputName=Scale_Y, inputType=FunctionInput_Scalar, sortPriority=4, previewValue=<1 0 0 0>}
  n4 = AppendVector(A=@n2, B=@n3)
  MaterialExpressionFunctionInput_4 = FunctionInput() {bUsePreviewValueAsDefault=true, description=Rad, inputName=Rotation, inputType=FunctionInput_Scalar, sortPriority=5, previewValue=<0 0 0 0>}
  n5 = Custom!(Angle=@MaterialExpressionFunctionInput_4) {
   code="float s=0.0;\r\nfloat c=1.0;\r\nsincos(Angle, s, c);\r\nreturn float2(s,c);"
   description=SinCos
   outputType=CMOT_Float2
   inputs [{inputName=Angle, input {expression=&self:MaterialExpressionFunctionInput_4}}]
  }
  n6 = FunctionInput() {bUsePreviewValueAsDefault=true, inputName=Offset_X, inputType=FunctionInput_Scalar, sortPriority=1, previewValue=<0 0 1 1>}
  n7 = FunctionInput() {bUsePreviewValueAsDefault=true, inputName=Offset_Y, inputType=FunctionInput_Scalar, sortPriority=2, previewValue=<0 0 1 1>}
  MaterialExpressionAdd_0 = Add(
   A=MaterialFunctionCall!(UV=Multiply(A=@n1, B=@n4), RotationCenter=Constant2Vector(), SinCosAngle=@n5) {materialFunction=&/MF_GLTF_RotateV2}.Result
   B=AppendVector(A=@n6, B=@n7)
  )
  MaterialExpressionFunctionInput_10 = FunctionInput() {bUsePreviewValueAsDefault=true, inputName=TilingMethod, inputType=FunctionInput_Vector2, sortPriority=6}
  n8 = Custom!(UV=@MaterialExpressionAdd_0, TilingMethod=@MaterialExpressionFunctionInput_10) {
   code="DDX = ddx(UV);\r\nDDY = ddy(UV);\r\n\r\nfloat Result_U = 0;\r\nfloat Result_V = 0;\r\n\r\n//U component's tiling:\r\nif (TilingMethod.x < 1)\r\n{\r\n    //WRAP\r\n    //Texture sampler is set to Shared: Wrap\r\n    //do nothing\r\n    Result_U = fmod(UV.x, 1.0);\r\n    if (Result_U < 0) Result_U = 1.0 + Result_U;\r\n}\r\nelse if (TilingMethod.x < 2)\r\n{\r\n    //CLAMP\r\n    //Texture sampler is set to Shared: Clamp\r\n    //do nothing\r\n    Result_U = UV.x;\r\n}\r\nelse \r\n{\r\n    //MIRROR\r\n    float LerpValue_U = fmod(floor(abs(UV.x)), 2);\r\n    float ToLerp_U = fmod(abs(UV.x), 1);\r\n\r\n    Result_U = lerp(ToLerp_U, 1.0-ToLerp_U, LerpValue_U);\r\n}\r\n\r\n//V component's tiling:\r\nif (TilingMethod.y < 1)\r\n{\r\n    //WRAP\r\n    //Texture sampler is set to Shared: Wrap\r\n    //do nothing\r\n    Result_V = fmod(UV.y, 1.0);\r\n    if (Result_V < 0) Result_V = 1.0 + Result_V;\r\n}\r\nelse if (TilingMethod.y < 2)\r\n{\r\n    //CLAMP\r\n    //Texture sampler is set to Shared: Clamp\r\n    //do nothing\r\n    Result_V = UV.y;\r\n}\r\nelse \r\n{\r\n    //MIRROR\r\n    float LerpValue_V = fmod(floor(abs(UV.y)), 2);\r\n    float ToLerp_V = fmod(abs(UV.y), 1);\r\n\r\n    Result_V = lerp(ToLerp_V, 1.0-ToLerp_V, LerpValue_V);\r\n}\r\n\r\nreturn float2(Result_U, Result_V);"
   description=ApplyTiling
   outputType=CMOT_Float2
   showCode=true
   additionalOutputs [{outputName=DDX, outputType=CMOT_Float2}, {outputName=DDY, outputType=CMOT_Float2}]
   inputs [
    {inputName=UV, input {expression=&self:MaterialExpressionAdd_0}}
    {inputName=TilingMethod, input {expression=&self:MaterialExpressionFunctionInput_10}}
   ]
  }
  FunctionOutput(None=@n8.return) {outputName=UVs}
  FunctionOutput(None=@n8.DDX) {outputName=DDX, sortPriority=1}
  FunctionOutput(None=@n8.DDY) {outputName=DDY, sortPriority=2}
 ]}
