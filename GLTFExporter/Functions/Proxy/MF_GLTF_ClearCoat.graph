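# Material function for glTF proxy materials in GLTFExporter that outputs the clear coat value from exposed Clear Coat parameters.
# MF_GLTF_UVTransform builds UVs from index, offset, scale, rotation and tiling parameters, with X and Y of offset and scale both wired to the R channel.
# Those UVs feed a derivative-mip texture sample, defaulting to linear white, whose red channel is multiplied by Clear Coat Factor.
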MaterialFunction /GLTFExporter/Functions/Proxy/MF_GLTF_ClearCoat
graph {nodes [
  n1 = VectorParameter() {group=Clear Coat, parameterName=Clear Coat UV Offset, sortPriority=3}
  n2 = VectorParameter() {group=Clear Coat, parameterName=Clear Coat UV Scale, sortPriority=4, defaultValue=<rgba 1 1 0 0>}
  n3 = MaterialFunctionCall!(
   TexCoord=ScalarParameter() {group=Clear Coat, parameterName=Clear Coat UV Index, sliderMax=7, sortPriority=2}
   Offset_X=@n1.R
   Offset_Y=@n1.R
   Scale_X=@n2.R
   Scale_Y=@n2.R
   Rotation=ScalarParameter() {group=Clear Coat, parameterName=Clear Coat UV Rotation, sliderMax=6.2832, sortPriority=5}
   TilingMethod=VectorParameter() {group=Clear Coat, parameterName=Clear Coat Tiling Method, sortPriority=3}.RGB
  ) {materialFunction=&/MF_GLTF_UVTransform}
  n4 = TextureSampleParameter2D!(UVs=@n3.UVs, "DDX(UVs)"=@n3.UVs, "DDY(UVs)"=@n3.UVs, Apply View MipBias) {
   group=Clear Coat
   mipValueMode=TMVM_Derivative
   parameterName=Clear Coat Texture
   samplerSource=SSM_Clamp_WorldGroupSettings
   samplerType=SAMPLERTYPE_LinearColor
   sortPriority=1
   texture=/GLTFExporter/Textures/Proxy/T_GLTF_DefaultWhiteLinear
   channelNames {a="", b="", g="", r=ClearCoat}
  }
  FunctionOutput(None=Multiply(A=@n4.ClearCoat, B=ScalarParameter() {group=Clear Coat, parameterName=Clear Coat Factor, sliderMax=1})) {outputName=ClearCoat}
 ]}
