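# Material function for glTF export proxy materials that exposes Sheen parameters and outputs Sheen Color (Fuzz Color) and Sheen Roughness (Cloth).
# Each output samples its own texture, white by default, through MF_GLTF_UVTransform with parameterized UV index, offset, scale, rotation and tiling.
# Color multiplies texture RGB by a factor; roughness multiplies texture alpha by a factor.
# Both X and Y of offset and scale read the vector parameters' R channel.
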MaterialFunction /GLTFExporter/Functions/Proxy/MF_GLTF_Sheen
graph {nodes [
  n1 = VectorParameter() {group=Sheen, parameterName=Sheen Color UV Offset, sortPriority=3}
  n2 = VectorParameter() {group=Sheen, parameterName=Sheen Color UV Scale, sortPriority=4, defaultValue=<rgba 1 1 0 0>}
  n3 = MaterialFunctionCall!(
   TexCoord=ScalarParameter() {group=Sheen, parameterName=Sheen Color UV Index, sliderMax=7, sortPriority=2}
   Offset_X=@n1.R
   Offset_Y=@n1.R
   Scale_X=@n2.R
   Scale_Y=@n2.R
   Rotation=ScalarParameter() {group=Sheen, parameterName=Sheen Color UV Rotation, sliderMax=6.2832, sortPriority=5}
   TilingMethod=VectorParameter() {group=Sheen, parameterName=Sheen Color Tiling Method, sortPriority=3}.RGB
  ) {materialFunction=&/MF_GLTF_UVTransform}
  n4 = TextureSampleParameter2D!(UVs=@n3.UVs, "DDX(UVs)"=@n3.UVs, "DDY(UVs)"=@n3.UVs, Apply View MipBias) {
   group=Sheen
   mipValueMode=TMVM_Derivative
   parameterName=Sheen Color Texture
   samplerSource=SSM_Clamp_WorldGroupSettings
   sortPriority=1
   texture=/GLTFExporter/Textures/Proxy/T_GLTF_DefaultWhite
  }
  FunctionOutput(None=Multiply(A=@n4.RGB, B=VectorParameter() {group=Sheen, parameterName=Sheen Color Factor, sortPriority=0}.RGB)) {description=Fuzz Color, outputName=Sheen Color}
  n5 = VectorParameter() {group=Sheen, parameterName=Sheen Roughness UV Offset, sortPriority=9}
  n6 = VectorParameter() {group=Sheen, parameterName=Sheen Roughness UV Scale, sortPriority=10, defaultValue=<rgba 1 1 0 0>}
  n7 = MaterialFunctionCall(
   TexCoord=ScalarParameter() {group=Sheen, parameterName=Sheen Roughness UV Index, sliderMax=7, sortPriority=8}
   Offset_X=@n5.R
   Offset_Y=@n5.R
   Scale_X=@n6.R
   Scale_Y=@n6.R
   Rotation=ScalarParameter() {group=Sheen, parameterName=Sheen Roughness UV Rotation, sliderMax=6.2832, sortPriority=11}
   TilingMethod=VectorParameter() {group=Sheen, parameterName=Sheen Roughness Tiling Method, sortPriority=9}.RGB
  ) {materialFunction=&MF_GLTF_UVTransform}
  n8 = TextureSampleParameter2D(UVs=@n7.UVs, "DDX(UVs)"=@n7.UVs, "DDY(UVs)"=@n7.UVs) {
   group=Sheen
   mipValueMode=TMVM_Derivative
   parameterName=Sheen Roughness Texture
   samplerSource=SSM_Clamp_WorldGroupSettings
   sortPriority=7
   texture=&T_GLTF_DefaultWhite
  }
  FunctionOutput(None=Multiply(A=@n8.A, B=ScalarParameter() {group=Sheen, parameterName=Sheen Roughness Factor, sortPriority=6})) {description=Cloth, outputName=Sheen Roughness, sortPriority=1}
 ]}
