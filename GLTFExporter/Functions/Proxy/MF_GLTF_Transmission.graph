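# Material function in the GLTF exporter's proxy set that turns glTF transmission parameters into an Opacity output.
# A Transmission Texture, defaulting to white, is sampled with UVs from MF_GLTF_UVTransform, driven by exposed index, offset, scale and rotation parameters.
# Its red channel is multiplied by Transmission Factor and inverted, so opacity equals one minus transmission.
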MaterialFunction /GLTFExporter/Functions/Proxy/MF_GLTF_Transmission
graph {nodes [
  n1 = VectorParameter() {group=Transmission, parameterName=Transmission UV Scale, sortPriority=4, defaultValue=<rgba 1 1 0 0>}
  n2 = TextureSampleParameter2D(UVs=MaterialFunctionCall!(
    UV Index=ScalarParameter() {group=Transmission, parameterName=Transmission UV Index, sliderMax=7, sortPriority=2}
    UV Offset=VectorParameter() {group=Transmission, parameterName=Transmission UV Offset, sortPriority=3}.RGB
    UV Scale=@n1.RGB
    UV Rotation=ScalarParameter() {group=Transmission, parameterName=Transmission UV Rotation, sliderMax=6.2832, sortPriority=5}
   ) {materialFunction=&/MF_GLTF_UVTransform}.UVs) {group=Transmission, parameterName=Transmission Texture, sortPriority=1, texture=/GLTFExporter/Textures/Proxy/T_GLTF_DefaultWhite}
  n3 = ScalarParameter() {group=Transmission, parameterName=Transmission Factor, sortPriority=0}
  FunctionOutput(None=OneMinus(None=Multiply(A=@n2.R, B=@n3))) {description="=1-Transmission", outputName=Opacity}
 ]}
